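# Physics asset for the F_MED_Assemble_R backpack cosmetic, defining bodies and joints for the pack and its two dangling tubes.
# A single capsule covers the pack bone, and left and right chains of three small capsules hang from it along the dyn_tube bones.
# Joints allow wide swing at the pack and tighter swing down each chain, with little twist throughout.
# The right chain's middle joint keeps the wide swing limit, unlike the left.
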PhysicsAsset /BRCosmetics/Accessories/FORT_Backpacks/Backpack_F_MED_Assemble_R/Meshes/F_MED_Assemble_R_Pack_Physics
bodies {
 dyn_tube_l_01 [{shapeType=Capsule, length=2.9282, radius=1.3311, center=<2.6452 0 0.0001>, rotation=<rot 90 0 0>}]
 dyn_tube_l_02 [{shapeType=Capsule, length=2.9282, radius=1.3311, center=<2.8493 0 -0.0002>, rotation=<rot 90 0 0>}]
 dyn_tube_l_03 [{shapeType=Capsule, length=2.9282, radius=1.3311, center=<2.8878 0 0.0001>, rotation=<rot 90 0 0>}]
 dyn_tube_r_01 [{shapeType=Capsule, length=2.9282, radius=1.3311, center=<2.6452 0 0.0001>, rotation=<rot 90 0 0>}]
 dyn_tube_r_02 [{shapeType=Capsule, length=2.9282, radius=1.3311, center=<2.9342 0 0>, rotation=<rot 90 0 0>}]
 dyn_tube_r_03 [{shapeType=Capsule, length=2.9282, radius=1.3311, center=<3.0186 0 0>, rotation=<rot 90 0 0>}]
 pack [{shapeType=Capsule, length=8.4346, radius=2.6481, center=<-0.006 -4.7356 -2.5206>, rotation=<rot 5.9007 -90 0.0509>}]
}
constraints [
 {bone1Name=dyn_tube_l_01, bone2Name=pack, swing1LimitDegrees=45, swing2LimitDegrees=45, twistLimitDegrees=5}
 {bone1Name=dyn_tube_l_02, bone2Name=dyn_tube_l_01, swing1LimitDegrees=10, swing2LimitDegrees=10, twistLimitDegrees=5}
 {bone1Name=dyn_tube_l_03, bone2Name=dyn_tube_l_02, swing1LimitDegrees=10, swing2LimitDegrees=10, twistLimitDegrees=5}
 {bone1Name=dyn_tube_r_01, bone2Name=pack, swing1LimitDegrees=45, swing2LimitDegrees=45, twistLimitDegrees=5}
 {bone1Name=dyn_tube_r_02, bone2Name=dyn_tube_r_01, swing1LimitDegrees=45, swing2LimitDegrees=45, twistLimitDegrees=5}
 {bone1Name=dyn_tube_r_03, bone2Name=dyn_tube_r_02, swing1LimitDegrees=10, swing2LimitDegrees=10, twistLimitDegrees=5}
]
