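# Animation Blueprint for the F_MED_Assemble_R backpack cosmetic, adding secondary physics motion to its skeleton.
# A RigidBody node simulates from the reference pose in root-bone space with custom gravity and damped component motion, its alpha lowered while bIsTargeting is true.
# Translation constraints then tie each dyn_tube 03 bone to its same-side 01 bone, partially on the left and fully on the right.
# The event graph holds no logic.
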AnimBlueprint /BRCosmetics/Accessories/FORT_Backpacks/Backpack_F_MED_Assemble_R/Meshes/F_MED_Assemble_R_Pack_AnimBP
target_skeleton=&/F_MED_Assemble_R_Pack_Skeleton
graphs [
 AnimGraph {nodes [
   n1 = RigidBody!(
    ComponentPose=Local To Component(LocalPose=Local Space Ref Pose())
    Alpha=SelectFloat(A=0.4, B=1, bPickA=Get bIsTargeting())
    OverrideWorldGravity:Vector="0,0,-98"
    ExternalForce
   ) -> (Pose) {
    bClampLinearTranslationLimitToRefPose=true
    bOverrideWorldGravity=true
    overridePhysicsAsset=&/F_MED_Assemble_R_Pack_Physics
    simulationSpace=BaseBoneSpace
    alphaBoolBlend {blendOption=Linear}
    baseBoneRef {boneName=root}
    componentLinearAccScale=<0.15 0.15 0.05>
    componentLinearVelScale=<0.1 0.1>
    overrideWorldGravity {z=-380}
   }
   n2 = "Constraint - dyn_tube_l_03 (T)"(ComponentPose=@n1, Alpha=0.5, ConstraintWeights_0=1) {
    alphaBoolBlend {blendOption=Linear}
    boneToModify {boneName=dyn_tube_l_03}
    constraintSetup [{transformType=Translation, targetBone {boneName=dyn_tube_l_01}}]
    constraintWeights [1]
   }
   n3 = "Constraint - dyn_tube_r_03 (T)"(ComponentPose=@n2, Alpha=1, ConstraintWeights_0=1) {
    alphaBoolBlend {blendOption=Linear}
    boneToModify {boneName=dyn_tube_r_03}
    constraintSetup [{transformType=Translation, targetBone {boneName=dyn_tube_r_01}}]
    constraintWeights [1]
   }
   Output Pose(Result=Component To Local(ComponentPose=@n3)) {layerGroup=None}
  ]}
 EventGraph {nodes [
   TryGetPawnOwner()
   Event BlueprintUpdateAnimation()
  ]}
]
variables []
